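# Parent material for the RubySnapper gadget's skeletal prop, adding animated lamps and a spinning-wheel effect to the shared 2023 anime character shader with cloth shading.
# MaterialCurveB switches lamps between off, sine blinking and on, with time-of-day-adjusted intensity.
# The wheel texture uses offset, scaled UVs with panning radial-noise warp and sphere masks, alongside a MaterialCurveA spin flag.
# Channel masks composite both effects into lit and shaded colors between the shader's Pre and Post functions.
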Material /BRCosmetics/Animation/Props/Gadget_RubySnapper/Materials/M_RubySnapper_Gadget_Parent
graph {
 properties {
  bAutomaticallySetUsageInEditor=false
  bUseMaterialAttributes=true
  bUsedWithMeshDeformer=true
  bUsedWithSkeletalMesh=true
  bUsedWithStaticMesh=false
  pixelDepthOffsetMode=PDOM_Legacy
  shadingModel=MSM_Cloth
  usedShadingModels=Cloth
 }
 nodes [
  n1 = MaterialFunctionCall!(CustomUV) {materialFunction=/Game/Global/Materials/Character_Parent_2023/Anime_Parent_2023/MaterialFunctions/MFC_AnimeShader_2023_Pre}
  n2 = MaterialFunctionCall!(
   CustomLightVector
   OverrideNormal_WS
   STT=@n1.STT
   Normal_Map=@n1.STT
   DFL=NamedRerouteUsage()
   FX=NamedRerouteDeclaration(None=@n1.FX) {name=FXIn, nodeColor=<rgba 1 0.0234 0 1>}
  ) {materialFunction=/Game/Global/Materials/Character_Parent_2023/Anime_Parent_2023/MaterialFunctions/MFC_AnimeShader_2023_Calculations}
  MaterialFunctionCall!(
   Shading_Gradient=@n2.Shading_Gradient
   Threshold_Gradient=@n2.Shading_Gradient
   Specular_Gradient=@n2.Shading_Gradient
   Sheen_Mask=@n2.Shading_Gradient
   Thickness=@n2.Shading_Gradient
   DFL_Mask=@n2.Shading_Gradient
   FX=@n2.Shading_Gradient
   Metallic_Gradient=@n2.Shading_Gradient
   RimLight_Gradient=@n2.Shading_Gradient
   NormalBuffer=@n2.Shading_Gradient
   --------
   CINE_LightSources=@n2.Shading_Gradient
   CINE_LightColor=@n2.Shading_Gradient
   ---------
   Color_Lit=NamedRerouteUsage()
   Color_Shaded=NamedRerouteUsage()
   SRM=@n1.SRM
   Emissive=@n1.SRM
   DFL_Color=@n1.SRM
   Specular_Color=@n1.SRM
   RimLight_Color=@n1.SRM
  ) {materialFunction=/Game/Global/Materials/Character_Parent_2023/Anime_Parent_2023/MaterialFunctions/MFC_AnimeShader_2023_Post}
  NamedRerouteDeclaration(None=@n1.Color_Lit) {name=ColorLitIn, nodeColor=<rgba 0.9453 0 1 1>}
  NamedRerouteDeclaration(None=@n1.Color_Shaded) {name=ColorShadedIn, nodeColor=<rgba 0 1 0.625 1>}
  NamedRerouteDeclaration(None=ScalarParameter() {desc=Lights Control, group=RubySnap Prop, parameterName=MaterialCurveB, sliderMax=2}) {name=LightsControl, nodeColor=<rgba 1 0 0.5651 1>}
  n3 = NamedRerouteUsage()
  MaterialExpressionScalarParameter_2 = ScalarParameter() {defaultValue=0.8, group=RubySnap Prop, parameterName=LightsOffIntensity}
  MaterialExpressionAdd_5 = Add(A=@MaterialExpressionScalarParameter_2, B=NamedRerouteUsage())
  MaterialExpressionNamedRerouteUsage_25 = NamedRerouteUsage()
  n4 = Switch!(
   SwitchValue=Ceil(None=NamedRerouteUsage())
   Default=@MaterialExpressionScalarParameter_2
   LightsOff=@MaterialExpressionScalarParameter_2
   LightsBlink=@MaterialExpressionAdd_5
   LightsOn=@MaterialExpressionNamedRerouteUsage_25
  ) {inputs [
    {
     inputName=LightsOff
     input {
      expression=&self:MaterialExpressionScalarParameter_2
      inputName=None
      mask=0
      maskA=0
      maskB=0
      maskG=0
      maskR=0
      outputIndex=0
     }
    }
    {
     inputName=LightsBlink
     input {
      expression=&self:MaterialExpressionAdd_5
      inputName=None
      mask=0
      maskA=0
      maskB=0
      maskG=0
      maskR=0
      outputIndex=0
     }
    }
    {
     inputName=LightsOn
     input {
      expression=&self:MaterialExpressionNamedRerouteUsage_25
      inputName=None
      mask=0
      maskA=0
      maskB=0
      maskG=0
      maskR=0
      outputIndex=0
     }
    }
   ]}
  n5 = Multiply(A=@n3, B=Multiply(A=@n4) {constB=0.8}) {constB=0.5}
  n6 = NamedRerouteUsage()
  n7 = NamedRerouteUsage()
  n8 = Multiply(A=NamedRerouteUsage(), B=NamedRerouteUsage())
  NamedRerouteDeclaration(None=LinearInterpolate(A=LinearInterpolate(A=@n3, B=@n5, Alpha=@n6), B=Multiply(A=@n7) {constB=0.5}, Alpha=@n8)) {name=ColorShadedOut, nodeColor=<rgba 0.3828 1 0 1>}
  NamedRerouteDeclaration(None=ScalarParameter() {desc=Is Prop Spinning?, group=RubySnap Prop, parameterName=MaterialCurveA, sliderMax=1}) {name=IsSpinning, nodeColor=<rgba 1 0.0938 0 1>}
  n9 = Time()
  n10 = ScalarParameter() {defaultValue=3.5, group=RubySnap Prop, parameterName=LightsSpeed}
  n11 = Sine(None=Add(A=Multiply(A=@n9, B=@n10), B=NamedRerouteUsage()))
  n12 = ScalarParameter() {defaultValue=1.5, group=RubySnap Prop, parameterName=LightsIntensity}
  n13 = Multiply(A=@n12) {constB=0.9}
  n14 = MaterialFunctionCall(Value_Morning=@n12, Value_Day=@n12, Value_Evening=@n13, Value_Night=@n13) {materialFunction=&Scalar_Time_of_Day}
  NamedRerouteDeclaration(None=Multiply(A=Saturate(None=@n11), B=NamedRerouteDeclaration(None=@n14.Result) {name=LightsIntensity, nodeColor=<rgba 0.2656 0 1 1>}) {constB=3}) {name=BlinkingLights, nodeColor=<rgba 0 0.2266 1 1>}
  n15 = Multiply(A=Subtract(A=NamedRerouteUsage()) {constB=0.5}, B=ScalarParameter() {defaultValue=0.2, group=RubySnap Prop, parameterName=WarpIntensity}) {constB=0.15}
  n16 = SphereMask(A=NamedRerouteUsage(), B=Constant(), Radius=ScalarParameter() {defaultValue=0.36, group=RubySnap Prop, parameterName=WheelWarpFadeAdjust}) {attenuationRadius=0.1, hardnessPercent=0}
  n17 = TextureSampleParameter2D(UVs=Add(A=NamedRerouteUsage(), B=LinearInterpolate(A=@n15, Alpha=@n16) {constB=0})) {
   group=RubySnap Prop
   parameterName=SpinningWheel_Map
   samplerSource=SSM_Clamp_WorldGroupSettings
   texture=/BRCosmetics/Animation/Props/Gadget_RubySnapper/Textures/T_RubySnapper_Emote_Prop_Wheel
  }
  NamedRerouteDeclaration(None=@n17.RGB) {name=OutWheel, nodeColor=<rgba 1 0 0.2344 1>}
  NamedRerouteDeclaration(None=@n1.DFL) {name=DFLIn, nodeColor=<rgba 0 1 0.7891 1>}
  n18 = NamedRerouteUsage()
  n19 = NamedRerouteUsage()
  NamedRerouteDeclaration(None=LinearInterpolate(A=NamedRerouteUsage(), Alpha=Multiply(A=@n18, B=@n19)) {constB=0}) {name=DFLOut, nodeColor=<rgba 1 0.4688 0 1>}
  n20 = VectorParameter() {group=RubySnap Prop, parameterName=WheelOffset, defaultValue=<rgba -0.093 -0.2373 0 1>}
  n21 = MaterialFunctionCall(
   UVs=Add(A=TextureCoordinate(), B=ComponentMask(None=@n20.RGB) {g=true, r=true})
   Texture Scale=ScalarParameter() {defaultValue=0.4, group=RubySnap Prop, parameterName=WheelScale}
  ) {materialFunction=&ScaleUVsByCenter}
  NamedRerouteDeclaration(None=Add(A=NamedRerouteDeclaration(None=@n21.UVs) {name=UVs, nodeColor=<rgba 0 1 0.0859 1>}, B=Constant2Vector() {g=-0.498, r=-0.5025}) {constB=-0.5}) {name=UVs Centered, nodeColor=<rgba 0.7344 1 0 1>}
  n22 = NamedRerouteUsage()
  NamedRerouteDeclaration(None=Subtract(
    A=ChannelMaskParameter(None=@n22) {group=RubySnap Prop, maskChannel=Green, parameterName=WheelMask, defaultValue=<rgba 0 1 0 0>}.R
    B=NamedRerouteUsage()
   )) {name=WheelMask, nodeColor=<rgba 1 0 0.6328 1>}
  n23 = NamedRerouteUsage()
  n24 = MaterialFunctionCall(Vector or UVs=@n23) {materialFunction=&VectorToRadialValue}
  n25 = DDX(Value=@n23)
  n26 = DDY(Value=@n23)
  n27 = TextureSample!(UVs=Panner(Coordinate=@n24.Radial Coordinates) {speedY=-0.75}, Tex, "DDX(UVs)"=@n25, "DDY(UVs)"=@n26, Apply View MipBias) {mipValueMode=TMVM_Derivative, samplerType=SAMPLERTYPE_Grayscale, texture=/Game/Global/Textures/Default/Noise/LowResBlurredNoiseNormalized}
  n28 = TextureSample(UVs=Multiply(A=Panner(Coordinate=@n24.Radial Coordinates) {speedX=0.1, speedY=-0.3}) {constB=2}, "DDX(UVs)"=@n25, "DDY(UVs)"=@n26) {mipValueMode=TMVM_Derivative, samplerType=SAMPLERTYPE_Grayscale, texture=&LowResBlurredNoiseNormalized}
  NamedRerouteDeclaration(None=Saturate(None=Add(A=@n27.R, B=@n28.R))) {name=RadialNoise, nodeColor=<rgba 1 0 0.2344 1>}
  NamedRerouteDeclaration(None=SphereMask(A=NamedRerouteUsage(), B=Constant(), Radius=ScalarParameter() {defaultValue=0.092, group=RubySnap Prop, parameterName=CenterMaskRadius}) {attenuationRadius=0.0865, hardnessPercent=99}) {name=CenterPiece, nodeColor=<rgba 0.9453 0 1 1>}
  n29 = ChannelMaskParameter(None=@n22) {group=RubySnap Prop, parameterName=LampsMask}
  NamedRerouteDeclaration(None=Ceil(None=NamedRerouteDeclaration(None=@n29.R) {name=LampsGradient, nodeColor=<rgba 0 1 0.3672 1>})) {name=LampsMask, nodeColor=<rgba 1 0.0469 0 1>}
  n30 = NamedRerouteUsage()
  n31 = Multiply(A=@n30, B=@n4) {constB=0.5}
  n32 = NamedRerouteUsage()
  NamedRerouteDeclaration(None=LinearInterpolate(A=LinearInterpolate(A=@n30, B=@n31, Alpha=@n32), B=NamedRerouteUsage(), Alpha=@n8)) {name=ColorLitOut, nodeColor=<rgba 0 0.2734 1 1>}
 ]
}
