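# Animation Blueprint for the RubySnapper gadget prop skeleton, able to collapse its two crow bodies and wheel frame through animation curves.
# DefaultSlot montage output feeds three chained Modify Bone nodes that replace scale with zero, using curve values as alpha.
# Crow curves are cross-wired: Scale_Crow1 drives JNT_Crow2_Body and Scale_Crow2 drives JNT_Crow1_Body, while SCale_Wheel drives JNT_WheelFrame.
# The EventGraph holds no connected logic, and the Scale variable is unused.
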AnimBlueprint /BRCosmetics/Animation/Props/Gadget_RubySnapper/Meshes/RubySnapper_Gadget_Skeleton_AnimBP
target_skeleton=&/RubySnapper_Gadget_Skeleton
graphs [
 AnimGraph {nodes [
   n1 = Slot 'DefaultSlot'()
   n2 = "Transform (Modify) Bone - Bone: JNT_Crow1_Body"(
    ComponentPose="Transform (Modify) Bone - Bone: JNT_Crow2_Body"(
     ComponentPose=Local To Component(LocalPose=@n1)
     Alpha=GetCurveValue(CurveName=Scale_Crow1)
     Translation="0,0,0"
     Rotation="0,0,0"
     Scale="0,0,0"
    ) {scaleMode=BMM_Replace, boneToModify {boneName=JNT_Crow2_Body}, scale=<0 0 0>}
    Alpha=GetCurveValue(CurveName=Scale_Crow2)
    Translation="0,0,0"
    Rotation="0,0,0"
    Scale="0,0,0"
   ) {scaleMode=BMM_Replace, boneToModify {boneName=JNT_Crow1_Body}, scale=<0 0 0>}
   n3 = GetCurveValue(CurveName=SCale_Wheel)
   Output Pose(Result=Component To Local(ComponentPose="Transform (Modify) Bone - Bone: JNT_WheelFrame"(ComponentPose=@n2, Alpha=@n3, Translation="0,0,0", Rotation="0,0,0", Scale="0,0,0") {scaleMode=BMM_Replace, boneToModify {boneName=JNT_WheelFrame}, scale=<0 0 0>}))
  ]}
 EventGraph {nodes [
   TryGetPawnOwner()
   Event BlueprintUpdateAnimation()
  ]}
]
variables [Scale {type=Vector, default=<1 1 1>}]
